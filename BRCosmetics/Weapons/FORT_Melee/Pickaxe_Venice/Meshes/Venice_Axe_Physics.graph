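# Physics asset for the Venice pickaxe cosmetic, defining dangling feather and string chains on the two-handed axe.
# Small capsules form a feather chain ending in a detail bone and a string chain with a short branch off its fifth bone.
# Both chains attach to the handle capsule with free swing and twist.
# Downstream joints have varying angular limits, tightest on the string branch.
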PhysicsAsset /BRCosmetics/Weapons/FORT_Melee/Pickaxe_Venice/Meshes/Venice_Axe_Physics
bodies {
 Melee_TwoHanded_Handle [{shapeType=Capsule, length=18.1307, radius=2.9201, center=<0 0 -4.938>}]
 feather_det_dyn [{shapeType=Capsule, length=1, radius=1, shapeName=feather_det_dyn_capsule}]
 feather_dyn_01 [{shapeType=Capsule, length=1, radius=1, shapeName=feather_dyn_01_capsule}]
 feather_dyn_02 [{shapeType=Capsule, length=1, radius=1, shapeName=feather_dyn_02_capsule}]
 feather_dyn_03 [{shapeType=Capsule, length=1, radius=1, shapeName=feather_dyn_03_capsule}]
 feather_dyn_04 [{shapeType=Capsule, length=1, radius=1, shapeName=feather_dyn_04_capsule}]
 feather_dyn_05 [{
   shapeType=Capsule
   length=6.8219
   radius=1.9683
   shapeName=feather_dyn_05_capsule
   center=<6.233 -0.4287 0.6496>
   rotation=<rot 90 6.5366 9.1384>
  }]
 string_Dyn_01 [{shapeType=Capsule, length=1, radius=1, shapeName=string_Dyn_01_capsule}]
 string_Dyn_02 [{shapeType=Capsule, length=1, radius=1, shapeName=string_Dyn_02_capsule}]
 string_Dyn_03 [{shapeType=Capsule, length=1, radius=1, shapeName=string_Dyn_03_capsule}]
 string_Dyn_04 [{shapeType=Capsule, length=8.9869, radius=1, shapeName=string_Dyn_04_capsule, center=<2.6195 0 0>, rotation=<rot 90 0 0>}]
 string_Dyn_05 [{shapeType=Capsule, length=1, radius=1, shapeName=string_Dyn_05_capsule, center=<0.5 0 0>}]
 string_Dyn_06 [{shapeType=Capsule, length=1, radius=1, shapeName=string_Dyn_06_capsule, center=<0.5 0 0>}]
 string_Dyn_b_01 [{shapeType=Capsule, length=0.5, radius=0.5, shapeName=string_Dyn_b_01_capsule}]
 string_Dyn_b_02 [{shapeType=Capsule, length=0.5, radius=0.5, shapeName=string_Dyn_b_02_capsule, center=<2 1 0>}]
}
constraints [
 {
  bone1Name=string_Dyn_01
  bone2Name=Melee_TwoHanded_Handle
  swing1LimitDegrees=45
  swing1Motion=Free
  swing2LimitDegrees=45
  swing2Motion=Free
  twistLimitDegrees=45
  twistMotion=Free
 }
 {bone1Name=string_Dyn_02, bone2Name=string_Dyn_01, swing1LimitDegrees=80, swing2LimitDegrees=80, twistLimitDegrees=80}
 {bone1Name=string_Dyn_03, bone2Name=string_Dyn_02, swing1LimitDegrees=180, swing2LimitDegrees=180, twistLimitDegrees=180}
 {bone1Name=string_Dyn_04, bone2Name=string_Dyn_03, swing1LimitDegrees=120, swing2LimitDegrees=120, twistLimitDegrees=120}
 {bone1Name=string_Dyn_05, bone2Name=string_Dyn_04, swing1LimitDegrees=120, swing2LimitDegrees=120, twistLimitDegrees=120}
 {bone1Name=string_Dyn_06, bone2Name=string_Dyn_05, swing1LimitDegrees=45, swing2LimitDegrees=45, twistLimitDegrees=45}
 {bone1Name=string_Dyn_b_01, bone2Name=string_Dyn_05, swing1LimitDegrees=15, swing2LimitDegrees=10, twistLimitDegrees=1}
 {bone1Name=string_Dyn_b_02, bone2Name=string_Dyn_b_01, swing1LimitDegrees=15, swing2LimitDegrees=10, twistLimitDegrees=1}
 {
  bone1Name=feather_dyn_01
  bone2Name=Melee_TwoHanded_Handle
  swing1LimitDegrees=45
  swing1Motion=Free
  swing2LimitDegrees=45
  swing2Motion=Free
  twistLimitDegrees=45
  twistMotion=Free
 }
 {bone1Name=feather_dyn_02, bone2Name=feather_dyn_01, swing1LimitDegrees=45, swing2LimitDegrees=45, twistLimitDegrees=45}
 {bone1Name=feather_dyn_03, bone2Name=feather_dyn_02, swing1LimitDegrees=30, swing2LimitDegrees=30, twistLimitDegrees=30}
 {bone1Name=feather_dyn_04, bone2Name=feather_dyn_03, swing1LimitDegrees=45, swing2LimitDegrees=45, twistLimitDegrees=45}
 {bone1Name=feather_dyn_05, bone2Name=feather_dyn_04, swing1LimitDegrees=45, swing2LimitDegrees=45, twistLimitDegrees=45}
 {bone1Name=feather_det_dyn, bone2Name=feather_dyn_05, swing1LimitDegrees=45, swing2LimitDegrees=45, twistLimitDegrees=1}
]
